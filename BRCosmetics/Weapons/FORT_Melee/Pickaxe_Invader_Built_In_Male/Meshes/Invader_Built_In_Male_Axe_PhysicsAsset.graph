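# Physics asset for the male Invader "Built In" pickaxe axe mesh, defining its collision bodies and joint limits.
# Three capsule bodies cover the lowerarm_twist_01_r, chain8 and chain24 bones, with the largest on the forearm twist bone.
# Two constraints with equal swing and twist angle limits link them in series, from lowerarm_twist_01_r through chain8 to chain24.
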PhysicsAsset /BRCosmetics/Weapons/FORT_Melee/Pickaxe_Invader_Built_In_Male/Meshes/Invader_Built_In_Male_Axe_PhysicsAsset
bodies {
 chain24 [{shapeType=Capsule, length=11.0497, radius=7.7153, center=<0.8223 6.3372 -0.111>, rotation=<rot 0 50.3217 -88.0876>}]
 chain8 [{shapeType=Capsule, length=14.0567, radius=6.4679, center=<-0.9748 -10.9545 -0.1322>, rotation=<rot 0 151.6137 -89.3759>}]
 lowerarm_twist_01_r [{shapeType=Capsule, length=37.6046, radius=22.5527, center=<-12.7788 -5.4365 -1.0194>, rotation=<rot 0 -156.6492 -84.8303>}]
}
constraints [
 {bone1Name=chain8, bone2Name=lowerarm_twist_01_r, swing1LimitDegrees=45, swing2LimitDegrees=45, twistLimitDegrees=45}
 {bone1Name=chain24, bone2Name=chain8, swing1LimitDegrees=45, swing2LimitDegrees=45, twistLimitDegrees=45}
]
