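# Anim Blueprint for the Invader built-in male pickaxe, animating the axe through equip, loop and unequip.
# A state machine drives those states, with the axe animation layered from saw_01 down, then lowerarm_r is constrained to root.
# Each update reads IsEquipped from the owning pickaxe and derives a first-equip flag that restarts the equip animation after unequip.
# The twist-bone transform and debug print are unconnected.
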AnimBlueprint /BRCosmetics/Weapons/FORT_Melee/Pickaxe_Invader_Built_In_Male/Meshes/Invader_Built_In_Male_Axe_Skeleton_AnimBP
target_skeleton=&/Invader_Built_In_Male_Axe_Skeleton
graphs [
 AnimGraph {nodes [
   "Transform (Modify) Bone - Bone: lowerarm_twist_01_r"(Alpha=1, Translation="0.000,1.2,-0.000", Rotation="0,0,0", Scale="1,1,1") {
    translationMode=BMM_Additive
    translationSpace=BCS_BoneSpace
    alphaBoolBlend {blendOption=Linear}
    boneToModify {boneName=lowerarm_twist_01_r}
    translation {y=1.2}
   }
   n1 = Sequence Player 'Invader_Built_In_Male_Axe_Anim'() {sequence=/BRCosmetics/Weapons/FORT_Melee/Pickaxe_Invader_Built_In_Male/Animations/Invader_Built_In_Male_Axe_Anim}
   AnimGraphNode_StateMachine_0 = Invader_BuiltIn_Axe_SM() {is=title AnimGraphNode_StateMachine}
   n2 = "Constraint - lowerarm_r (P)"(
    ComponentPose=Local To Component(LocalPose=Layered blend per bone(BasePose=@AnimGraphNode_StateMachine_0, BlendPoses_0=@n1) {layerSetup [{branchFilters [{boneName=saw_01}]}]})
    Alpha=1
    ConstraintWeights_0=1
   ) {
    alphaBoolBlend {blendOption=Linear}
    boneToModify {boneName=lowerarm_r}
    constraintSetup [{offsetOption=None, targetBone {boneName=root}}]
    constraintWeights [1]
   }
   Output Pose(Result=Component To Local(ComponentPose=@n2)) {layerGroup=None}
  ]}
 Invader_BuiltIn_Axe_SM {
  entry=Start
  path=AnimGraph.AnimGraphNode_StateMachine_0.Invader_BuiltIn_Axe_SM
  ids=id
  nodes [
   AnimStateNode_0 = Loop() {is=title AnimStateNode}
   AnimStateNode_1 = Start() {is=title AnimStateNode}
   AnimStateNode_3 = Unequip() {is=title AnimStateNode}
   AnimStateTransitionNode_0 = AnimStateTransitionNode() {bAutomaticRuleBasedOnSequencePlayerInState=true}
   AnimStateTransitionNode_2 = AnimStateTransitionNode() {crossfadeDuration=0}
   AnimStateTransitionNode_3 = AnimStateTransitionNode() {crossfadeDuration=0}
  ]
  links [
   {
    dst=Unequip
    rule=AnimGraph.AnimGraphNode_StateMachine_0.Invader_BuiltIn_Axe_SM.AnimStateTransitionNode_2.Transition
    src=Loop
    via=AnimStateTransitionNode_2
   }
   {
    dst=Loop
    rule=AnimGraph.AnimGraphNode_StateMachine_0.Invader_BuiltIn_Axe_SM.AnimStateTransitionNode_0.Transition
    src=Start
    via=AnimStateTransitionNode_0
   }
   {
    dst=Start
    rule=AnimGraph.AnimGraphNode_StateMachine_0.Invader_BuiltIn_Axe_SM.AnimStateTransitionNode_3.Transition
    src=Unequip
    via=AnimStateTransitionNode_3
   }
  ]
 }
 Loop {
  path=AnimGraph.AnimGraphNode_StateMachine_0.Invader_BuiltIn_Axe_SM.AnimStateNode_0.Loop
  nodes [Output Animation Pose(Result=Sequence Player 'Invader_BuiltIn_Axe_Equip_Loop'() {sequence=/BRCosmetics/Weapons/FORT_Melee/Pickaxe_Invader_Built_In_Male/Animations/Invader_BuiltIn_Axe_Equip_Loop})]
 }
 Start {
  path=AnimGraph.AnimGraphNode_StateMachine_0.Invader_BuiltIn_Axe_SM.AnimStateNode_1.Start
  nodes [Output Animation Pose(Result=Sequence Player 'Invader_BuiltIn_Axe_Equip_Anim'() {sequence=/BRCosmetics/Weapons/FORT_Melee/Pickaxe_Invader_Built_In_Male/Animations/Invader_BuiltIn_Axe_Equip_Anim})]
 }
 Unequip {
  path=AnimGraph.AnimGraphNode_StateMachine_0.Invader_BuiltIn_Axe_SM.AnimStateNode_3.Unequip
  nodes [Output Animation Pose(Result=Sequence Player 'Invader_Builtin_Axe_UnEquip'() {bLoopAnimation=false, sequence=/BRCosmetics/Weapons/FORT_Melee/Pickaxe_Invader_Built_In_Male/Animations/Invader_Builtin_Axe_UnEquip})]
 }
 Transition {path=AnimGraph.AnimGraphNode_StateMachine_0.Invader_BuiltIn_Axe_SM.AnimStateTransitionNode_0.Transition, nodes [Result()]}
 Transition {
  path=AnimGraph.AnimGraphNode_StateMachine_0.Invader_BuiltIn_Axe_SM.AnimStateTransitionNode_2.Transition
  nodes [Result(bCanEnterTransition=NOT Boolean(A=Get IsEquipped() -> (IsEquipped:bool)))]
 }
 Transition {
  path=AnimGraph.AnimGraphNode_StateMachine_0.Invader_BuiltIn_Axe_SM.AnimStateTransitionNode_3.Transition
  nodes [Result(bCanEnterTransition=AND Boolean(A=Get IsInitialEquipped() -> (IsInitialEquipped:bool), B=Get IsEquipped()))]
 }
 EventGraph {nodes [
   Event BlueprintUpdateAnimation()
   > n1 = Cast To FortWeaponPickaxeAthena!(execute, Object=GetOwningActor()) -> (then, CastFailed, AsFort Weapon Pickaxe Athena:Fort Weapon Pickaxe Athena Object Reference) {is=title K2Node_DynamicCast}
   > n2 = Set IsEquipped(IsEquipped:bool=IsEquipped!(self:Fort Weapon Object Reference=@n1) -> (ReturnValue:bool)) -> (Output_Get:bool)
   > n3 = Branch(Condition=@n2)
   > n4 = Branch(Condition=Get IsAlreadyEquipped() -> (IsAlreadyEquipped:bool))
   > Set IsInitialEquipped(IsInitialEquipped:bool=false) -> (Output_Get:bool)
   Set IsAlreadyEquipped(execute=@n4.else, IsAlreadyEquipped:bool=true) -> (Output_Get:bool)
   Set IsAlreadyEquipped(execute=@n3.else, IsAlreadyEquipped=false)
   > Set IsInitialEquipped(IsInitialEquipped=true)
   TryGetPawnOwner()
   n5 = Get IsAlreadyEquipped()
   n6 = Get IsInitialEquipped()
   PrintString!(
    execute
    InString=Append!(
     A="IsAlreadyEquipped = "
     B="To String (Boolean)"(InBool=@n5)
     C:String="\r\n"
     D:String="IsInitialEquipped = "
     E:String="To String (Boolean)"(InBool=@n6)
    ) -> (ReturnValue)
    bPrintToScreen
    bPrintToLog
    TextColor
    Duration
    Key
   ) -> (then)
  ]}
]
variables [
 IsAlreadyEquipped {type=boolean}
 IsEquipped {type=boolean}
 IsInitialEquipped {type=boolean, default=true}
]
