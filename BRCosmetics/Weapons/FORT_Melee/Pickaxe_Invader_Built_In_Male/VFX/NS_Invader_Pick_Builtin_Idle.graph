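# Niagara idle effect for the Invader built-in pickaxe, adding sparks and animated warm-colored lights to the blade and chainsaw.
# Initial_Spark emits short-lived glints on randomized loops, with extra spawning during non-left swings, and feeds positions to Sparks, which throws dragged streaks in a cone.
# A system FN_Integrator value scales spark rate and streak speed.
# Two light emitters cycle color curves at fixed points, spawning only when UsesParticleLights passes its threshold.
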NiagaraSystem /BRCosmetics/Weapons/FORT_Melee/Pickaxe_Invader_Built_In_Male/VFX/NS_Invader_Pick_Builtin_Idle
dependencies {
 usedDataInterfaces [
  &NiagaraDataInterfaceColorCurve
  &NiagaraDataInterfaceArrayFloat3
  &NiagaraDataInterfaceCurve
  &NiagaraDataInterfaceArrayColor
  &NiagaraDataInterfaceParticleRead
  &NiagaraDataInterfaceVector2DCurve
 ]
 usedDynamicInputs []
 usedModules [
  &self:SystemSpawnScript.SystemScriptSource.SystemScriptGraph.NiagaraNodeAssignment_0.SetVariables_#1
  &SystemState
  &FN_Integrator
 ]
}
emitters {
 Initial_Spark {
  emitterUpdateScript [
   &EmitterState(
    Delay First Loop Only=true
    Life Cycle Mode=::NewEnumerator1
    Loop Delay=2
    Loop Duration=&Lerp_Float(
     A=0
     B=3
     Alpha=&FloatFromCurve(
      FloatCurve=DataInterface() {propertyValues {Curve {keys [
          {
           arriveTangent=0.2834
           arriveTangentWeight=0.8967
           interpMode=RCIM_Cubic
           leaveTangent=0.2834
           leaveTangentWeight=0.8967
           tangentMode=RCTM_User
           tangentWeightMode=RCTWM_WeightedBoth
           time=0
           value=0
          }
          {
           arriveTangent=0.3758
           arriveTangentWeight=0.9371
           interpMode=RCIM_Cubic
           leaveTangent=0.3758
           leaveTangentWeight=0.9371
           tangentMode=RCTM_User
           tangentWeightMode=RCTWM_WeightedBoth
           time=1
           value=1
          }
         ]}}}
      CurveIndex=&RandomRangeFloat(Minimum=0, Maximum=1)
     )
    )
    Recalculate Duration Each Loop=true
    UseLoopDelay=true
   )
   &SpawnRate(SpawnRate=&Lerp_Float(A=0, B=50, Alpha=$System.FN_Integrator001.Value), Use Spawn Probability=false)
   &SpawnBurst_Instantaneous(
    Age:NiagaraFloat=&OneMinusFloat(Float=$System.FN_Integrator001.Value)
    Spawn Count=&RandomRangeInt(Minimum=15, Maximum=25, Evaluation Type=::NewEnumerator1)
    Spawn Time=&Multiply_Float(A=$Engine.DeltaTime, B=1.5)
    Use Spawn Probability=false
   )
   &SpawnBurst_Instantaneous(Spawn Count=1, Use Spawn Probability=false) {module=SpawnBurst_Instantaneous001}
   &SpawnRate(
    SpawnRate=&MakeCustomFloatFromBool(
     Bool=&Boolean_AndOperation(Bool1=$User.bIsSwinging, Bool2=&Boolean_NotOperation(Bool=$User.bIsSwingingLeft))
     True Float=10
     False Float=0
    )
    Use Spawn Probability=false
   ) {module=SpawnRate001}
  ]
  particleSpawnScript [
   &InitializeParticle(
    Color=<rgba 10 3.0725 0 1>
    Lifetime Max=0.025
    Lifetime Min=0.05
    Lifetime Mode=::NewEnumerator1
    Sprite Rotation Mode=::NewEnumerator2
    Sprite Size Mode=::NewEnumerator1
    Sprite UV Mode=::NewEnumerator3
    Uniform Sprite Size Max=5
    Uniform Sprite Size Min=2
    UsePositionOffset=false
    Write Material Random=true
   )
   &CameraOffset(Camera Offset Amount=2)
   &SphereLocation(
    MaskSpawnGroup=false
    Offset=<-47 1.5 -0.5>
    Override Seed=false
    Sphere Radius=1.5
    Surface Only=false
    UseNonUniformScale=false
    UseOffset=true
    UseSphereOrientation=false
   )
   /SetVariables(Particles.Corona Cam offset=10, Particles.Corona Sprite Size:Vector2f=<25 25>, Particles.Corona color:LinearColor=<rgba 10 3.0725 0 0.1>) {isSetParameters=true}
  ]
  particleUpdateScript [
   &ParticleState()
   &ScaleSpriteSize(Scale Factor=&Vector2DFromFloat(Value=&FloatFromCurve(FloatCurve=DataInterface() {propertyValues {Curve {keys [<at 0 0>,<at 0.7 1.5>,<at 1 0.2>]}}})))
   &SolveForcesAndVelocity(Clamp Velocity=false, Limit Acceleration=false)
   &GenerateDeathEvent() {enabled=false}
  ]
  rendererClasses [&NiagaraSpriteRendererProperties]
  renderers [
   {rendererClass=&NiagaraSpriteRendererProperties}
   {rendererClass=&NiagaraSpriteRendererProperties, rendererIndex=1}
  ]
 }
 Lights_Blade {
  emitterSpawnScript [/SetVariables(Emitter.RainbowCurve=DataInterface() {propertyValues {
      AlphaCurve {keys [<at 0 1>]}
      BlueCurve {keys [
        <at 0 0 RCIM_Cubic>
        {
         arriveTangentWeight=0.6094
         interpMode=RCIM_Cubic
         leaveTangentWeight=0.6094
         tangentMode=RCTM_User
         tangentWeightMode=RCTWM_WeightedBoth
         time=0.5213
         value=0
        }
        <at 1 0>
       ]}
      GreenCurve {keys [
        {interpMode=RCIM_Cubic, leaveTangentWeight=0.1992, time=0, value=0.399}
        {
         arriveTangentWeight=0.2818
         interpMode=RCIM_Cubic
         leaveTangentWeight=0.2818
         tangentMode=RCTM_User
         tangentWeightMode=RCTWM_WeightedBoth
         time=0.5213
         value=0
        }
        <at 1 0.399 RCIM_Cubic>
       ]}
      RedCurve {keys [<at 0 1 RCIM_Cubic>,<at 0.5213 1 RCIM_Cubic>,<at 1 1>]}
     }}) {isSetParameters=true, module=SetVariables_#2}]
  emitterUpdateScript [
   &EmitterState(Inactive Response=::NewEnumerator1, Life Cycle Mode=::NewEnumerator1, Loop Duration=1, UseLoopDelay=false)
   &SpawnBurst_Instantaneous(
    Age:NiagaraFloat=$Emitter.Age
    Spawn Count=&MakeCustomIntFromBool(Bool=&SetBoolByFloatComparison(A=$User.UsesParticleLights, B=0.5), True INT=2, False INT=0)
    Use Spawn Probability=false
   )
  ]
  particleSpawnScript [&InitializeParticle(
    Lifetime=3.25
    Position=&SelectVectorFromArray(
     Direct Array Index=&ReturnExecIndex()
     Vector Selection Array=DataInterface() {propertyValues {FloatData [<-38 -34.609 -4.583>, <-38 -19.1728 -2.457>]}}
    )
    Position Mode=::NewEnumerator1
    Sprite Size Mode=::NewEnumerator3
    Uniform Sprite Size=25
    UsePositionOffset=false
    Write Material Random=true
   ) {module=InitializeParticle001}]
  particleUpdateScript [
   &ParticleState(
    Kill Particles When Lifetime Has Elapsed=false
    Let Infinitely Lived Particles Die When Emitter Deactivates=true
    Loop Particles Lifetime=true
   )
   &Color(
    Color=&ColorFromCurve(
     ColorCurve=$Emitter.RainbowCurve
     CurveIndex=&Frac_Float(Float=&Add_Float(
       A=&Multiply_Float(A=$Emitter.Age, B=0.25)
       B=&ReturnNormalizedExecIndex(Normalized Index Scale=0.45)
      ))
    )
    Scale Color=&VectorFromFloat(Value=5)
   )
  ]
  rendererClasses [&NiagaraLightRendererProperties]
  renderers [{rendererClass=&NiagaraLightRendererProperties, rendererIndex=0}]
 }
 Lights_Chain_Saw {
  emitterSpawnScript [/SetVariables(Emitter.RainbowCurve=DataInterface() {propertyValues {
      AlphaCurve {keys [<at 0.4247 1>]}
      BlueCurve {keys [<at 0 0 RCIM_Cubic>,<at 0.3514 0 RCIM_Cubic>,<at 0.6371 0 RCIM_Cubic>,<at 1 0 RCIM_Cubic>]}
      GreenCurve {keys [
        <at 0 0.196 RCIM_Cubic>
        {arriveTangent=-0.3076, interpMode=RCIM_Cubic, leaveTangent=-0.3076, time=0.2, value=0.3}
        {arriveTangent=-0.13, interpMode=RCIM_Cubic, leaveTangent=-0.13, time=0.6371, value=0}
        {arriveTangent=1.2752, interpMode=RCIM_Cubic, leaveTangent=1.2752, tangentMode=RCTM_User, time=1, value=0.196}
       ]}
      RedCurve {keys [<at 0 1 RCIM_Cubic>,<at 0.3514 1 RCIM_Cubic>,<at 0.6371 1 RCIM_Cubic>,<at 1 1 RCIM_Cubic>]}
     }}) {isSetParameters=true, module=SetVariables_#2}]
  emitterUpdateScript [
   &EmitterState(Inactive Response=::NewEnumerator1, Life Cycle Mode=::NewEnumerator1, Loop Duration=1, UseLoopDelay=false)
   &SpawnBurst_Instantaneous(
    Age:NiagaraFloat=$Emitter.Age
    Spawn Count=&MakeCustomIntFromBool(Bool=&SetBoolByFloatComparison(A=$User.UsesParticleLights, B=0.5), True INT=2, False INT=0)
    Use Spawn Probability=false
   )
  ]
  particleSpawnScript [&InitializeParticle(
    Color=&SelectLinearColorFromArray(
     Array Sampling Mode=::NewEnumerator1
     Direct Array Index=&ReturnExecIndex()
     Color Selection Array=DataInterface() {propertyValues {ColorData [<rgba 1 1 0 1>, <rgba 1 0 0.0686 1>]}}
     Random Seed=?
    )
    Lifetime=3.25
    Position=&SelectVectorFromArray(
     Direct Array Index=&ReturnExecIndex()
     Vector Selection Array=DataInterface() {propertyValues {FloatData [<-42.1321 8.5179 0.6788>, <-42.2744 20.9724 2.0945>]}}
    )
    Position Mode=::NewEnumerator1
    Sprite Size Mode=::NewEnumerator3
    Uniform Sprite Size=25
    UsePositionOffset=false
    Write Material Random=true
   )]
  particleUpdateScript [
   &ParticleState(
    Kill Particles When Lifetime Has Elapsed=false
    Let Infinitely Lived Particles Die When Emitter Deactivates=true
    Loop Particles Lifetime=true
   )
   &Color(
    Color=&HueShiftLinearColor(
     Hue=&Add_Float(A=&OneMinusFloat(Float=&Multiply_Float(A=$Particles.NormalizedAge, B=3)), B=0.25)
     Saturation=1
     Value=1
     Linear Color=$Particles.Initial.Color
    )
    Scale Color=&VectorFromFloat(Value=5)
   ) {enabled=false}
   &Color(
    Color=&ColorFromCurve(
     ColorCurve=$Emitter.RainbowCurve
     CurveIndex=&Frac_Float(Float=&Add_Float(
       A=&Multiply_Float(A=$Emitter.Age, B=2)
       B=&ReturnNormalizedExecIndex(Normalized Index Scale=-0.4)
      ))
    )
    Scale Color=&VectorFromFloat(Value=5)
   ) {module=Color001}
  ]
  rendererClasses [&NiagaraLightRendererProperties, &NiagaraSpriteRendererProperties]
  renderers [
   {rendererClass=&NiagaraLightRendererProperties, rendererIndex=0}
   {rendererClass=&NiagaraSpriteRendererProperties, rendererIndex=1}
  ]
 }
 Sparks {
  emitterUpdateScript [
   &EmitterState(Life Cycle Mode=::NewEnumerator1, Loop Duration=1.15, UseLoopDelay=false)
   &SpawnParticlesFromOtherEmitter(
    Attribute Reader=DataInterface() {propertyValues {EmitterBinding {bindingMode=Other, emitterName=Initial_Spark}}}
    Calculate Spawn Rate Per Particle=false
    Spawn Rate=100
    Use Maximum Age=false
    Use Minimum Age=false
   )
  ]
  particleSpawnScript [
   &InitializeParticle(
    Color=<rgba 50 41.1997 10.578 1>
    Lifetime Max=0.2
    Lifetime Min=0.1
    Lifetime Mode=::NewEnumerator1
    Sprite Size=<1.5 12>
    Sprite Size Mode=::NewEnumerator4
    UsePositionOffset=false
    Write Material Random=true
   )
   &SampleParticlesFromOtherEmitter()
   &AddVelocityInCone(
    Cone Axis=<-0.5 1 0>
    Module.Use Velocity Falloff On Cone Axis=false
    Velocity Strength=&Lerp_Float(
     A=&RandomRangeFloat(Minimum=350, Maximum=450)
     B=&RandomRangeFloat(Minimum=900, Maximum=1200)
     Alpha=$System.FN_Integrator001.Value
    )
   )
   &InheritVelocity(Module.Limit Speed=true, Velocity Limit=1000000000, Velocity Scale=<0.18 0.18 0.18>)
  ]
  particleUpdateScript [
   &ParticleState()
   &ScaleSpriteSize(
    Non-Uniform Curve Sprite Scale=DataInterface() {propertyValues {
      XCurve {keys [
        {
         interpMode=RCIM_Cubic
         leaveTangent=-3.2585
         leaveTangentWeight=0.9206
         tangentMode=RCTM_Break
         tangentWeightMode=RCTWM_WeightedBoth
         time=0
         value=1
        }
        {arriveTangentWeight=0.3333, interpMode=RCIM_Cubic, tangentMode=RCTM_User, tangentWeightMode=RCTWM_WeightedBoth, time=1, value=0}
       ]}
      YCurve {keys [{interpMode=RCIM_Cubic, leaveTangent=-1.6549, leaveTangentWeight=0.383, tangentMode=RCTM_Break, time=0, value=1}, <at 1 0>]}
     }}
    Scale Sprite Size Mode=::NewEnumerator3
   )
   &Drag(Drag=15, Use Linear Drag=true, Use Rotational Drag=false)
   &SolveForcesAndVelocity(Clamp Velocity=false, Limit Acceleration=false)
  ]
  rendererClasses [&NiagaraSpriteRendererProperties]
  renderers [{rendererClass=&NiagaraSpriteRendererProperties}]
 }
}
summary {
 systemName=NS_Invader_Pick_Builtin_Idle
 emitters [
  {emitterName=Lights_Blade, rendererClasses [&NiagaraLightRendererProperties]}
  {emitterName=Initial_Spark, rendererClasses [&NiagaraSpriteRendererProperties]}
  {
   emitterName=Lights_Chain_Saw
   rendererClasses [&NiagaraLightRendererProperties, &NiagaraSpriteRendererProperties]
  }
  {emitterName=Sparks, rendererClasses [&NiagaraSpriteRendererProperties]}
 ]
 userVariables [
  User.bIsFrontendPreview {type=NiagaraBool}
  User.bIsSwinging {type=NiagaraBool}
  User.bIsSwingingLeft {type=NiagaraBool}
  User.HasImpacted {type=NiagaraBool, defaultValue=true}
  User.UsesParticleLights {type=NiagaraFloat, defaultValue=0, description=""}
 ]
}
