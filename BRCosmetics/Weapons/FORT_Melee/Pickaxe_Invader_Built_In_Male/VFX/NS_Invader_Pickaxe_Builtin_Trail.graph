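# Niagara system for the built-in male Invader pickaxe swing trail, drawing short-lived ribbons from skeletal mesh sockets.
# Left Blade and Right Chainsaw emitters are gated by FN_Timer comparisons and tinted by TrailColor; the chainsaw spawns two trails.
# A rainbow emitter, enabled through UseRainbowTrail, samples a color curve randomly.
# Alpha fades over life with time-of-day scaling, which isDarkVarient alters for the left blade; a sprite Fountain is disabled.
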NiagaraSystem /BRCosmetics/Weapons/FORT_Melee/Pickaxe_Invader_Built_In_Male/VFX/NS_Invader_Pickaxe_Builtin_Trail
dependencies {
 usedDataInterfaces [&NiagaraDataInterfaceCurve, &NiagaraDataInterfaceColorCurve]
 usedDynamicInputs [&Boolean_NotOperation]
 usedModules [&SystemState, &FN_Timer]
}
emitters {
 FN_AnimTrail_Swing_Left_Blade {
  emitterSpawnScript [&FN_RibbonIDOffsetUtility()]
  emitterUpdateScript [
   &EmitterState()
   &FN_SpawnAnimTrail(Enable:NiagaraBool=&SetBoolByFloatComparison(A=$System.FN_Timer.Value, B=0.5, Comparison Type=::NewEnumerator2))
  ]
  particleSpawnScript [
   &FN_InitializeAnimTrailBoneIndecies(First Bone or Socket Index=1, Second Bone or Socket Index=0)
   &FN_InitializeAnimTrail(Color=$User.TrailColor, Lifetime=0.15)
  ]
  particleUpdateScript [
   &FN_UpdateAnimTrail(Locked Particle RibbonWidth=1, SkeletalMesh=$User.SkeletalMesh) {module=FN_UpdateAnimTrail001}
   &ScaleColor(
    Scale Alpha=&Multiply_Float(
     A=&FloatFromCurve(
      FloatCurve=DataInterface() {propertyValues {Curve {keys [{arriveTangent=19.9617, interpMode=RCIM_Cubic, tangentMode=RCTM_User, time=0, value=1}, <at 1 0>]}}}
      Scale Curve=0.35
     )
     B=&Lerp_Float(
      A=&FNTimeOfDayScalar(Night Scale=0.3, Morning Scale=0.5, Evening Scale=0.5)
      B=&FNTimeOfDayScalar(Night Scale=2, Morning Scale=0.5, Evening Scale=0.5)
      Alpha=$User.isDarkVarient
     )
    )
    ScaleA=true
    ScaleRGB=true
   )
  ]
  rendererClasses [&NiagaraRibbonRendererProperties]
  renderers [{rendererClass=&NiagaraRibbonRendererProperties}]
 }
 FN_AnimTrail_Swing_Right_Chainsaw {
  emitterSpawnScript [&FN_RibbonIDOffsetUtility()]
  emitterUpdateScript [
   &EmitterState()
   &FN_SpawnAnimTrail(
    Enable:NiagaraBool=&SetBoolByFloatComparison(A=$System.FN_Timer.Value, B=0.5, Comparison Type=::NewEnumerator6)
    Number of Anim Trails=2
   )
   &SpawnPerUnit(
    Max Movement Threshold=0
    Module.Use Max Movement Threshold=true
    Module.Use Movement Tolerance=false
    Spawn Probability=0
    Spawn Spacing=1.3106511755113136e+19
    Use Spawn Probability=true
   )
  ]
  particleSpawnScript [
   &FN_InitializeAnimTrailBoneIndecies(First Bone or Socket Index=1, Second Bone or Socket Index=0)
   &FN_InitializeAnimTrailBoneIndecies(AnimTrail Selection=1) {module=FN_InitializeAnimTrailBoneIndecies001}
   &FN_InitializeAnimTrail(
    AnimTrailWidth=0.95
    Color=&LerpLinearColors(
     StartColor=&MultiplyLinearColor(Linear Color=$User.TrailColor, Scale Factor=<1 1 1 1>)
     EndColor=&MultiplyLinearColor(Linear Color=$User.TrailColor, Scale Factor=<1 1 1 0.25>)
     LerpFactor=&MakeFloatFromInt(Int=&SpawnGroup())
    )
    Lifetime=0.15
   )
  ]
  particleUpdateScript [
   &FN_UpdateAnimTrail(Locked Particle RibbonWidth=1, SkeletalMesh=$User.SkeletalMesh) {module=FN_UpdateAnimTrail001}
   &ScaleColor(
    Scale Alpha=&Multiply_Float(
     A=&FloatFromCurve(
      FloatCurve=DataInterface() {propertyValues {Curve {keys [{arriveTangent=19.9617, interpMode=RCIM_Cubic, tangentMode=RCTM_User, time=0, value=1}, <at 1 0>]}}}
      Scale Curve=0.35
     )
     B=&FNTimeOfDayScalar(Night Scale=0.3, Morning Scale=0.5, Evening Scale=0.5)
    )
    ScaleA=true
    ScaleRGB=true
   )
  ]
  rendererClasses [&NiagaraRibbonRendererProperties]
  renderers [{rendererClass=&NiagaraRibbonRendererProperties}]
 }
 FN_AnimTrail_rainbow {
  emitterSpawnScript [
   &FN_RibbonIDOffsetUtility()
   /SetVariables(Emitter.RainbowCurve=DataInterface() {propertyValues {
      AlphaCurve {keys [<at 0.4247 1>]}
      BlueCurve {keys [<at 0 0 RCIM_Cubic>,<at 0.3514 0 RCIM_Cubic>,<at 1 0 RCIM_Cubic>]}
      GreenCurve {keys [
        <at 0 0.399 RCIM_Cubic>
        {interpMode=RCIM_Cubic, leaveTangent=-0.6263, time=0.2, value=0.3}
        {interpMode=RCIM_Cubic, leaveTangent=0.125, time=1, value=0}
       ]}
      RedCurve {keys [<at 0 1 RCIM_Cubic>,<at 0.3514 1 RCIM_Cubic>,<at 1 1 RCIM_Cubic>]}
     }}) {isSetParameters=true}
  ]
  emitterUpdateScript [
   &EmitterState()
   &FN_SpawnAnimTrail(Enable:NiagaraBool=&SetBoolByFloatComparison(A=$User.UseRainbowTrail, B=0.5))
  ]
  particleSpawnScript [
   &FN_InitializeAnimTrailBoneIndecies(First Bone or Socket Index=1, Second Bone or Socket Index=0)
   &FN_InitializeAnimTrail(
    AnimTrailWidth=0.75
    Color=&HueShiftLinearColor(Hue=&Multiply_Float(A=$Engine.System.Age, B=50), Saturation=1, Value=1, Linear Color=<rgba 1 0 0 1>)
    Lifetime=0.15
   )
   &FN_InitializeAnimTrail(
    AnimTrailWidth=0.75
    Color=&ColorFromCurve(ColorCurve=$Emitter.RainbowCurve, CurveIndex=&RandomRangeFloat(Minimum=0, Maximum=1))
    Lifetime=0.15
   ) {module=FN_InitializeAnimTrail001}
  ]
  particleUpdateScript [
   &FN_UpdateAnimTrail(Locked Particle RibbonWidth=1, SkeletalMesh=$User.SkeletalMesh) {module=FN_UpdateAnimTrail001}
   &ScaleColor(
    Scale Alpha=&Multiply_Float(
     A=&FloatFromCurve(
      FloatCurve=DataInterface() {propertyValues {Curve {keys [{arriveTangent=19.9617, interpMode=RCIM_Cubic, tangentMode=RCTM_User, time=0, value=1}, <at 1 0>]}}}
      Scale Curve=0.35
     )
     B=&FNTimeOfDayScalar(Night Scale=0.3, Morning Scale=0.5, Evening Scale=0.5)
    )
    ScaleA=true
    ScaleRGB=true
   )
  ]
  rendererClasses [&NiagaraRibbonRendererProperties]
  renderers [{rendererClass=&NiagaraRibbonRendererProperties}]
 }
 Fountain {
  bEnabled=false
  emitterUpdateScript [
   &EmitterState(Life Cycle Mode=::NewEnumerator1, Loop Duration=2, UseLoopDelay=false)
   &SpawnRate(
    SpawnRate=&MakeCustomFloatFromBool(Bool=&SetBoolByFloatComparison(A=$User.UseRainbowTrail, B=0.5), True Float=1, False Float=0)
    Use Spawn Probability=false
   )
  ]
  particleSpawnScript [
   &InitializeParticle(
    Lifetime Max=1.75
    Lifetime Min=1.4
    Lifetime Mode=::NewEnumerator1
    Mass Mode=::NewEnumerator3
    Sprite Rotation Mode=::NewEnumerator2
    Sprite Size Mode=::NewEnumerator1
    Uniform Sprite Size Max=12
    Uniform Sprite Size Min=6
    UsePositionOffset=false
    Write Material Random=true
   )
   &SphereLocation(
    MaskSpawnGroup=false
    Override Seed=false
    Sphere Radius=8
    Surface Only=false
    UseNonUniformScale=false
    UseOffset=false
    UseSphereOrientation=false
   )
   &AddVelocityInCone(
    Cone Angle=20
    Cone Axis=<0 0 1>
    Module.Use Velocity Falloff On Cone Axis=false
    Velocity Distribution Along Cone Axis=0.75
    Velocity Strength=&RandomRangeFloat(Minimum=500, Maximum=850)
   )
  ]
  particleUpdateScript [
   &ParticleState()
   &GravityForce()
   &Drag(Drag=0.25, Use Linear Drag=true, Use Rotational Drag=false)
   &ScaleColor(
    Scale Alpha=&FloatFromCurve(FloatCurve=DataInterface() {propertyValues {Curve {keys [<at -0.0046 1.0113 RCIM_Cubic>,<at 1.0011 0.002 RCIM_Cubic>]}}})
    ScaleA=true
    ScaleRGB=true
   )
   &SolveForcesAndVelocity(Clamp Velocity=false, Limit Acceleration=false)
  ]
  rendererClasses [&NiagaraSpriteRendererProperties]
  renderers [{rendererClass=&NiagaraSpriteRendererProperties}]
 }
}
summary {
 systemName=NS_Invader_Pickaxe_Builtin_Trail
 emitters [
  {emitterName=FN_AnimTrail_Swing_Left_Blade, rendererClasses [&NiagaraRibbonRendererProperties]}
  {emitterName=FN_AnimTrail_rainbow, rendererClasses [&NiagaraRibbonRendererProperties]}
  {emitterName=FN_AnimTrail_Swing_Right_Chainsaw, rendererClasses [&NiagaraRibbonRendererProperties]}
  {bEnabled=false, emitterName=Fountain, rendererClasses [&NiagaraSpriteRendererProperties]}
 ]
 userVariables [
  User.bIsSwingingLeft {type=NiagaraBool}
  User.isDarkVarient {type=NiagaraFloat, defaultValue=0, description=""}
  User.RibbonIDOffset {type=NiagaraInt32}
  User.SkeletalMesh {
   type=NiagaraDataInterfaceSkeletalMesh
   defaultValue {
    struct=NiagaraExt_VariableValue_DataInterface
    value {dataInterface=&self:NiagaraDataInterfaceSkeletalMesh_0}
   }
  }
  User.TrailColor {type=LinearColor, description="", defaultValue {struct=LinearColor, value=<rgba 0.063 0.4999 0.5149 0.75>}}
  User.UseRainbowTrail {type=NiagaraFloat, defaultValue=0, description=""}
 ]
}
